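# Niagara system for the MadCommanderDark head cosmetic: smoke ribbons from left and right exhausts, plus a blue glow at each.
# Ribbons are few-particle Hermite splines anchored to skeletal mesh sampling, with ends offset against owner velocity for trailing and panned erosion materials.
# The mesh-and-sprite glow follows the same points, with color scaled by time of day.
# User.bisHeadHidden zeroes ribbon width and glow sprite size; FP exhaust variants are disabled.
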NiagaraSystem /BRCosmetics/Characters/Player/Male/Medium/Heads/M_MED_MadCommanderDark_Head/NS_MadCommanderDark_ExhaustSmoke
dependencies {usedDataInterfaces [&NiagaraDataInterfaceCurve], usedDynamicInputs [], usedModules [&SystemState]}
emitters {
 FP_LeftExhaust {
  bEnabled=false
  emitterSpawnScript [/Game/Effects/Niagara/Modules/Partition/FN_PartitionAttributes(Number of Bone Sets=1, Particles Per Partition=3)]
  emitterUpdateScript [
   &EmitterState(Life Cycle Mode=::NewEnumerator1, Loop Behavior=::NewEnumerator1, Loop Duration=1, UseLoopDelay=false)
   &SpawnBurst_Instantaneous(Spawn Count=$Emitter.FN_PartitionAttributes.Spawn Total, Use Spawn Probability=false)
  ]
  particleSpawnScript [
   /Niagara/Modules/Spawn/Initialization/InitializeParticle(
    Lifetime=5
    Module.Write Color=true
    Module.Write Lifetime=true
    Module.Write Mass=true
    Module.Write Material Random=true
    Module.Write Position=true
    Module.Write Scale=false
    Module.Write SpriteRotation=false
    Module.Write SpriteSize=true
    Sprite Size=<5 5>
   )
   /Game/Effects/Niagara/Modules/Partition/FN_Particle_partition_v5/FN_Particle_Partition_v5(Enable Ribbon ID=true)
   &InitializeRibbon(
    Module.Write Color=true
    Module.Write Lifetime=true
    Module.Write Mass=true
    Module.Write Material Random=true
    Module.Write Position=false
    Module.Write RibbonTwist=false
    Module.Write RibbonWidth=true
    Ribbon Width=&FloatFromCurve(
     FloatCurve=DataInterface() {propertyValues {Curve {keys [
         {interpMode=RCIM_Cubic, leaveTangent=-1.6682, tangentMode=RCTM_Break, time=0, value=1}
         {arriveTangent=-0.3052, interpMode=RCIM_Cubic, leaveTangent=-0.3052, tangentMode=RCTM_Break, time=0.2774, value=0.441}
         {arriveTangent=-0.3335, interpMode=RCIM_Cubic, leaveTangent=-0.3335, tangentMode=RCTM_Break, time=0.6089, value=0.7531}
         {interpMode=RCIM_Cubic, tangentMode=RCTM_Break, time=1.001, value=0.053}
        ]}}}
     CurveIndex=$Particles.Partition.NormalizedID
     Scale Curve=11
    )
   )
  ]
  particleUpdateScript [
   &ParticleState(Kill Particles When Lifetime Has Elapsed=false)
   &SkeletalMeshLocation(
    Bone / Socket Index=1
    Inherit Velocity=false
    MaskSpawnGroup=false
    Mesh Sampling Type=::NewEnumerator5
    Skeletal Mesh=$User.NewNiagaraDataInterfaceSkeletalMesh
    Socket Sampling Mode=::NewEnumerator5
   )
   /Game/Effects/Niagara/Modules/Curves/FN_HermiteCubicSpline_v1/FN_HermiteCubicSpline(
    Direction End:Vector3f=<2.8112 0 -5.7932>
    Direction Start:Vector3f=<-15.4964 -1.3118 -19.0608>
    Position Blend:NiagaraFloat=$Particles.Initial.Partition.NormalizedID
    Position End:Vector3f=&Add_Vector(
     A=$Output.SkeletalMeshLocation.SampledPosition
     B=&Add_Vector(
      A=<7.636 -0.9905 -14.5466>
      B=&Multiply_Vector(
       A=&TransformVector(Vector=$Engine.Owner.Velocity, Source Space=::World, Destination Space=::Local)
       B=&Multiply_Vector(A=<-1 -1 -1>, B=/Niagara/DynamicInputs/Math/Clamp_Vector(Vector:Vector3f=<0.02 0.05 0.015>, Min:Vector3f=<0 0 0>, Max:Vector3f=<1 1 1>))
      )
     )
    )
    Position Start:Vector3f=&Add_Vector(A=$Output.SkeletalMeshLocation.SampledPosition, B=<-0.483 0.04 0.57>)
    Write Particles Position=true
   )
   /Game/Effects/Niagara/Modules/Materials/FN_RibbonMaterial/FN_RibbonMaterial(
    Erosion Pan Speed U=1
    Erossion Scale U=2
    Macro Pan Speed=0.5
    Micro Pan Speed=0.2
    Opacity Erosion=0.2
    Opacity Pan U Speed=0.1
    Opacity Power=0.6
    Opacity Scale U=0.2
   )
  ]
  rendererClasses [&NiagaraRibbonRendererProperties]
  renderers [{rendererClass=&NiagaraRibbonRendererProperties}]
 }
 FP_RightExhaust {
  bEnabled=false
  emitterSpawnScript [&FN_PartitionAttributes(Number of Bone Sets=1, Particles Per Partition=3)]
  emitterUpdateScript [
   &EmitterState(Life Cycle Mode=::NewEnumerator1, Loop Behavior=::NewEnumerator1, Loop Duration=1, UseLoopDelay=false)
   &SpawnBurst_Instantaneous(Spawn Count=$Emitter.FN_PartitionAttributes.Spawn Total, Use Spawn Probability=false)
  ]
  particleSpawnScript [
   &InitializeParticle(
    Lifetime=5
    Module.Write Color=true
    Module.Write Lifetime=true
    Module.Write Mass=true
    Module.Write Material Random=true
    Module.Write Position=true
    Module.Write Scale=false
    Module.Write SpriteRotation=false
    Module.Write SpriteSize=true
    Sprite Size=<5 5>
   )
   &FN_Particle_Partition_v5(Enable Ribbon ID=true)
   &InitializeRibbon(
    Module.Write Color=true
    Module.Write Lifetime=true
    Module.Write Mass=true
    Module.Write Material Random=true
    Module.Write Position=false
    Module.Write RibbonTwist=false
    Module.Write RibbonWidth=true
    Ribbon Width=&FloatFromCurve(
     FloatCurve=DataInterface() {propertyValues {Curve {keys [
         {interpMode=RCIM_Cubic, leaveTangent=-1.6682, tangentMode=RCTM_Break, time=0, value=1}
         {arriveTangent=-0.3052, interpMode=RCIM_Cubic, leaveTangent=-0.3052, tangentMode=RCTM_Break, time=0.2774, value=0.441}
         {arriveTangent=-0.3335, interpMode=RCIM_Cubic, leaveTangent=-0.3335, tangentMode=RCTM_Break, time=0.6089, value=0.7531}
         {interpMode=RCIM_Cubic, tangentMode=RCTM_Break, time=1.001, value=0.053}
        ]}}}
     CurveIndex=$Particles.Partition.NormalizedID
     Scale Curve=11
    )
   )
  ]
  particleUpdateScript [
   &ParticleState(Kill Particles When Lifetime Has Elapsed=false)
   &SkeletalMeshLocation(
    Inherit Velocity=false
    MaskSpawnGroup=false
    Mesh Sampling Type=::NewEnumerator5
    Skeletal Mesh=$User.NewNiagaraDataInterfaceSkeletalMesh
    Socket Sampling Mode=::NewEnumerator5
   )
   &FN_HermiteCubicSpline(
    Direction End:Vector3f=<2.8112 0 5.7932>
    Direction Start:Vector3f=<-15.4964 -1.3118 19.0608>
    Position Blend:NiagaraFloat=$Particles.Initial.Partition.NormalizedID
    Position End:Vector3f=&Add_Vector(
     A=$Output.SkeletalMeshLocation.SampledPosition
     B=&Add_Vector(
      A=<7.636 -0.9905 14.5466>
      B=&Multiply_Vector(
       A=&TransformVector(Vector=$Engine.Owner.Velocity, Source Space=::World, Destination Space=::Local)
       B=&Multiply_Vector(A=<-1 -1 -1>, B=&Clamp_Vector(Vector:Vector3f=<0.02 0.05 0.015>, Min:Vector3f=<0 0 0>, Max:Vector3f=<1 1 1>))
      )
     )
    )
    Position Start:Vector3f=&Add_Vector(A=$Output.SkeletalMeshLocation.SampledPosition, B=<-0.483 0.04 -0.57>)
    Write Particles Position=true
   )
   &FN_RibbonMaterial(
    Erosion Pan Speed U=1
    Erossion Scale U=2
    Macro Pan Speed=0.5
    Micro Pan Speed=0.2
    Opacity Erosion=0.2
    Opacity Pan U Speed=0.1
    Opacity Power=0.6
    Opacity Scale U=0.2
   )
  ]
  rendererClasses [&NiagaraRibbonRendererProperties]
  renderers [{rendererClass=&NiagaraRibbonRendererProperties}]
 }
 Inner_Core_Glow {
  emitterUpdateScript [
   &EmitterState(Life Cycle Mode=::NewEnumerator1, Loop Behavior=::NewEnumerator1, UseLoopDelay=false)
   &SpawnBurst_Instantaneous(Spawn Count=2, Use Spawn Probability=false)
  ]
  particleSpawnScript [
   &InitializeParticle(
    Lifetime=5
    Mesh Scale=<0.2 0.2 0.2>
    Module.Write Color=true
    Module.Write Lifetime=true
    Module.Write Mass=true
    Module.Write Material Random=true
    Module.Write Position=true
    Module.Write Scale=false
    Module.Write SpriteRotation=false
    Module.Write SpriteSize=true
    Position=&SimulationPosition()
    Sprite Size=&Vector2DFromFloat(Value=8.7682)
   )
   &CameraOffset(Camera Offset Amount=2)
   &FN_SetSpawnGroup_v2(Number of Spawn Groups=2)
  ]
  particleUpdateScript [
   &ParticleState(Kill Particles When Lifetime Has Elapsed=false, Loop Particles Lifetime=true)
   &ScaleSpriteSize(
    Scale Sprite Size Mode=::NewEnumerator0
    Uniform Scale Factor=&MakeCustomFloatFromBool(Bool=$User.bisHeadHidden, True Float=0, False Float=1)
   )
   &Color(
    Color=<rgba 0 1.1256 10 1>
    Scale Color=&VectorFromFloat(Value=&FN_Select_Float_from_Time_Of_Day(Morning=0.5, Day=1, Evening=1, Night=0.5))
   )
   &SkeletalMeshLocation(
    Bone / Socket Index=$Particles.SpawnGroup
    Inherit Velocity=false
    MaskSpawnGroup=false
    Mesh Sampling Type=::NewEnumerator5
    Skeletal Mesh=$User.NewNiagaraDataInterfaceSkeletalMesh
    Socket Sampling Mode=::NewEnumerator5
   )
   /SetVariables(Particles.Position=&Add_Vector(
     A=$Output.SkeletalMeshLocation.SampledPosition
     B=&MakeCustomVectorFromBool(Bool=HlslExpression() {hlslExpression="Particles.SpawnGroup < 1"}, True Vector=<-0.0717 0.196 0.5251>, False Vector=<-0.0717 0.196 -0.5251>)
    )) {isSetParameters=true}
  ]
  rendererClasses [&NiagaraMeshRendererProperties, &NiagaraSpriteRendererProperties]
  renderers [
   {rendererClass=&NiagaraMeshRendererProperties, rendererIndex=0}
   {rendererClass=&NiagaraSpriteRendererProperties, rendererIndex=1}
  ]
 }
 LeftExhaust {
  emitterSpawnScript [&FN_PartitionAttributes(Number of Bone Sets=1, Particles Per Partition=3)]
  emitterUpdateScript [
   &EmitterState(Life Cycle Mode=::NewEnumerator1, Loop Behavior=::NewEnumerator1, Loop Duration Mode=::NewEnumerator1, UseLoopDelay=false)
   &SpawnBurst_Instantaneous(Spawn Count=$Emitter.FN_PartitionAttributes.Spawn Total, Use Spawn Probability=false)
  ]
  particleSpawnScript [
   &InitializeParticle(
    Lifetime=5
    Module.Write Color=true
    Module.Write Lifetime=true
    Module.Write Mass=true
    Module.Write Material Random=true
    Module.Write Position=true
    Module.Write Scale=false
    Module.Write SpriteRotation=false
    Module.Write SpriteSize=true
    Sprite Size=<5 5>
   )
   &FN_Particle_Partition_v5(Enable Ribbon ID=true)
   &InitializeRibbon(
    Module.Write Color=true
    Module.Write Lifetime=true
    Module.Write Mass=true
    Module.Write Material Random=true
    Module.Write Position=false
    Module.Write RibbonTwist=false
    Module.Write RibbonWidth=true
    Ribbon Width=&FloatFromCurve(
     FloatCurve=DataInterface() {propertyValues {Curve {keys [
         {interpMode=RCIM_Cubic, leaveTangent=-1.6682, tangentMode=RCTM_Break, time=0, value=1}
         {arriveTangent=-0.3052, interpMode=RCIM_Cubic, leaveTangent=-0.3052, tangentMode=RCTM_Break, time=0.2774, value=0.441}
         {arriveTangent=-0.3335, interpMode=RCIM_Cubic, leaveTangent=-0.3335, tangentMode=RCTM_Break, time=0.6089, value=0.7531}
         {interpMode=RCIM_Cubic, tangentMode=RCTM_Break, time=1.001, value=0.053}
        ]}}}
     CurveIndex=$Particles.Partition.NormalizedID
     Scale Curve=11
    )
   )
  ]
  particleUpdateScript [
   &ParticleState(Kill Particles When Lifetime Has Elapsed=false, Let Infinitely Lived Particles Die When Emitter Deactivates=true)
   &SkeletalMeshLocation(
    Bone / Socket Index=1
    Inherit Velocity=false
    MaskSpawnGroup=false
    Mesh Sampling Type=::NewEnumerator5
    Skeletal Mesh=$User.NewNiagaraDataInterfaceSkeletalMesh
    Socket Sampling Mode=::NewEnumerator5
   )
   &FN_HermiteCubicSpline(
    Direction End:Vector3f=<20 0 -39.7465>
    Direction Start:Vector3f=<-20 -1.3118 -5.8136>
    Position Blend:NiagaraFloat=$Particles.Initial.Partition.NormalizedID
    Position End:Vector3f=&Add_Vector(
     A=$Output.SkeletalMeshLocation.SampledPosition
     B=&Add_Vector(
      A=<13.1714 -0.9905 -12.8424>
      B=&Multiply_Vector(
       A=&TransformVector(Vector=$Engine.Owner.Velocity, Source Space=::World, Destination Space=::Local)
       B=&Multiply_Vector(A=<-1 -1 -1>, B=&Clamp_Vector(Vector:Vector3f=<0.02 0.05 0.015>, Min:Vector3f=<0 0 0>, Max:Vector3f=<1 1 1>))
      )
     )
    )
    Position Start:Vector3f=&Add_Vector(A=$Output.SkeletalMeshLocation.SampledPosition, B=<-0.483 0.04 0.57>)
    Write Particles Position=true
   )
   /Game/Effects/Niagara/Modules/Materials/FN_RibbonMaterial/FN_RibbonMaterial_v2(
    Erosion Pan Speed U=1
    Erossion Scale U=2
    Macro Pan Speed=0.5
    Micro Pan Speed=0.2
    Opacity Erosion=0.2
    Opacity Pan U Speed=0.1
    Opacity Power=0.6
    Opacity Scale U=0.2
   )
   &ScaleRibbonWidth(Ribbon Width Scale=&MakeCustomFloatFromBool(Bool=$User.bisHeadHidden, True Float=0, False Float=1))
  ]
  rendererClasses [&NiagaraRibbonRendererProperties]
  renderers [{rendererClass=&NiagaraRibbonRendererProperties}]
 }
 RightExhaust {
  emitterSpawnScript [&FN_PartitionAttributes(Number of Bone Sets=1, Particles Per Partition=3)]
  emitterUpdateScript [
   &EmitterState(Life Cycle Mode=::NewEnumerator1, Loop Behavior=::NewEnumerator1, Loop Duration Mode=::NewEnumerator1, UseLoopDelay=false)
   &SpawnBurst_Instantaneous(Spawn Count=$Emitter.FN_PartitionAttributes.Spawn Total, Use Spawn Probability=false)
  ]
  particleSpawnScript [
   &InitializeParticle(
    Lifetime=5
    Module.Write Color=true
    Module.Write Lifetime=true
    Module.Write Mass=true
    Module.Write Material Random=true
    Module.Write Position=true
    Module.Write Scale=false
    Module.Write SpriteRotation=false
    Module.Write SpriteSize=true
    Sprite Size=<5 5>
   )
   &FN_Particle_Partition_v5(Enable Ribbon ID=true)
   &InitializeRibbon(
    Module.Write Color=true
    Module.Write Lifetime=true
    Module.Write Mass=true
    Module.Write Material Random=true
    Module.Write Position=false
    Module.Write RibbonTwist=false
    Module.Write RibbonWidth=true
    Ribbon Width=&FloatFromCurve(
     FloatCurve=DataInterface() {propertyValues {Curve {keys [
         {interpMode=RCIM_Cubic, leaveTangent=-1.6682, tangentMode=RCTM_Break, time=0, value=1}
         {arriveTangent=-0.3052, interpMode=RCIM_Cubic, leaveTangent=-0.3052, tangentMode=RCTM_Break, time=0.2774, value=0.441}
         {arriveTangent=-0.3335, interpMode=RCIM_Cubic, leaveTangent=-0.3335, tangentMode=RCTM_Break, time=0.6089, value=0.7531}
         {interpMode=RCIM_Cubic, tangentMode=RCTM_Break, time=1.001, value=0.053}
        ]}}}
     CurveIndex=$Particles.Partition.NormalizedID
     Scale Curve=11
    )
   )
  ]
  particleUpdateScript [
   &ParticleState(Kill Particles When Lifetime Has Elapsed=false, Let Infinitely Lived Particles Die When Emitter Deactivates=true)
   &SkeletalMeshLocation(
    Inherit Velocity=false
    MaskSpawnGroup=false
    Mesh Sampling Type=::NewEnumerator5
    Skeletal Mesh=$User.NewNiagaraDataInterfaceSkeletalMesh
    Socket Sampling Mode=::NewEnumerator5
   )
   &FN_HermiteCubicSpline(
    Direction End:Vector3f=<20 0 39.7465>
    Direction Start:Vector3f=<-20 -1.3118 5.8136>
    Position Blend:NiagaraFloat=$Particles.Initial.Partition.NormalizedID
    Position End:Vector3f=&Add_Vector(
     A=$Output.SkeletalMeshLocation.SampledPosition
     B=&Add_Vector(
      A=<13.1714 -0.9905 12.8424>
      B=&Multiply_Vector(
       A=&TransformVector(Vector=$Engine.Owner.Velocity, Source Space=::World, Destination Space=::Local)
       B=&Multiply_Vector(A=<-1 -1 -1>, B=&Clamp_Vector(Vector:Vector3f=<0.02 0.05 0.015>, Min:Vector3f=<0 0 0>, Max:Vector3f=<1 1 1>))
      )
     )
    )
    Position Start:Vector3f=&Add_Vector(A=$Output.SkeletalMeshLocation.SampledPosition, B=<-0.483 0.04 -0.57>)
    Write Particles Position=true
   )
   &FN_RibbonMaterial_v2(
    Erosion Pan Speed U=1
    Erossion Scale U=2
    Macro Pan Speed=0.5
    Micro Pan Speed=0.2
    Opacity Erosion=0.2
    Opacity Pan U Speed=0.1
    Opacity Power=0.6
    Opacity Scale U=0.2
   )
   &ScaleRibbonWidth(Ribbon Width Scale=&MakeCustomFloatFromBool(Bool=$User.bisHeadHidden, True Float=0, False Float=1))
  ]
  rendererClasses [&NiagaraRibbonRendererProperties]
  renderers [{rendererClass=&NiagaraRibbonRendererProperties}]
 }
}
summary {
 systemName=NS_MadCommanderDark_ExhaustSmoke
 emitters [
  {emitterName=LeftExhaust, rendererClasses [&NiagaraRibbonRendererProperties]}
  {emitterName=RightExhaust, rendererClasses [&NiagaraRibbonRendererProperties]}
  {bEnabled=false, emitterName=FP_LeftExhaust, rendererClasses [&NiagaraRibbonRendererProperties]}
  {bEnabled=false, emitterName=FP_RightExhaust, rendererClasses [&NiagaraRibbonRendererProperties]}
  {
   emitterName=Inner_Core_Glow
   rendererClasses [&NiagaraMeshRendererProperties, &NiagaraSpriteRendererProperties]
  }
 ]
 userVariables [
  User.bisHeadHidden {type=NiagaraBool}
  User.NewNiagaraDataInterfaceSkeletalMesh {
   type=NiagaraDataInterfaceSkeletalMesh
   defaultValue {
    struct=NiagaraExt_VariableValue_DataInterface
    value {dataInterface=&self:NiagaraDataInterfaceSkeletalMesh_0}
   }
  }
  User.RibbonIDOffset {type=NiagaraInt32}
 ]
}
